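# Physics asset for the left shoe of the F_MED_Stands female medium body cosmetic, covering its dynamic shoelace bones.
# Each of the five bones has one capsule body.
# Constraints form two two-segment chains branching from the shoelace base: 1 to 2, and 3 to 4.
# All joints share the same narrow swing limits and tighter twist limit, restricting how far the laces can bend.
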PhysicsAsset /BRCosmetics/Characters/Player/Female/Medium/Bodies/F_MED_Stands/Meshes/F_MED_Stands_ShoeLeft_Physics
bodies {
 dyn_shoelace1_l [{
   shapeType=Capsule
   length=1.7713
   radius=0.8403
   shapeName=dyn_shoelace1_l_capsule
   center=<1.3235 -0.295 0.1866>
   rotation=<rot 0 89.2051 -81.3509>
  }]
 dyn_shoelace2_l [{
   shapeType=Capsule
   length=2.7429
   radius=1.2057
   shapeName=dyn_shoelace2_l_capsule
   center=<2.8124 -0.6481 -0.0163>
   rotation=<rot 0 -89.92 -80.3856>
  }]
 dyn_shoelace3_l [{
   shapeType=Capsule
   length=2.0561
   radius=1.0525
   shapeName=dyn_shoelace3_l_capsule
   center=<1.0934 -0.2937 -0.1417>
   rotation=<rot 0 -92.8829 -86.4227>
  }]
 dyn_shoelace4_l [{
   shapeType=Capsule
   length=2.867
   radius=1.3363
   shapeName=dyn_shoelace4_l_capsule
   center=<2.7895 -0.2588 0.3599>
   rotation=<rot 0 99.9701 -85.786>
  }]
 dyn_shoelace_base_l [{
   shapeType=Capsule
   length=1.2683
   radius=1.2097
   shapeName=dyn_shoelace_base_l_capsule
   center=<0.4723 -0.0959 0.1024>
   rotation=<rot 0 89.2878 -65.8794>
  }]
}
constraints [
 {bone1Name=dyn_shoelace1_l, bone2Name=dyn_shoelace_base_l, swing1LimitDegrees=20, swing2LimitDegrees=20, twistLimitDegrees=5}
 {bone1Name=dyn_shoelace2_l, bone2Name=dyn_shoelace1_l, swing1LimitDegrees=20, swing2LimitDegrees=20, twistLimitDegrees=5}
 {bone1Name=dyn_shoelace3_l, bone2Name=dyn_shoelace_base_l, swing1LimitDegrees=20, swing2LimitDegrees=20, twistLimitDegrees=5}
 {bone1Name=dyn_shoelace4_l, bone2Name=dyn_shoelace3_l, swing1LimitDegrees=20, swing2LimitDegrees=20, twistLimitDegrees=5}
]
