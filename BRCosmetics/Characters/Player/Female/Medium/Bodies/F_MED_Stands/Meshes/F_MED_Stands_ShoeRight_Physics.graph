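# Physics asset for the right shoe mesh of the F_MED_Stands female player body cosmetic, giving its shoelace bones bodies for dynamic motion.
# Five capsule bodies cover a base lace bone and two two-bone strands branching from it.
# Four constraints link each strand bone to its parent toward the base, with modest swing limits and a tighter twist limit on every joint.
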PhysicsAsset /BRCosmetics/Characters/Player/Female/Medium/Bodies/F_MED_Stands/Meshes/F_MED_Stands_ShoeRight_Physics
bodies {
 dyn_shoelace1_r [{
   shapeType=Capsule
   length=1.0283
   radius=0.7471
   shapeName=dyn_shoelace1_r_capsule
   center=<0.7334 -0.1395 -0.0118>
   rotation=<rot 0 81.4016 -89.4836>
  }]
 dyn_shoelace2_r [{
   shapeType=Capsule
   length=2.8642
   radius=1.2141
   shapeName=dyn_shoelace2_r_capsule
   center=<3.0779 -0.3523 0.759>
   rotation=<rot 0 89.3002 -73.9933>
  }]
 dyn_shoelace3_r [{
   shapeType=Capsule
   length=1.539
   radius=0.9652
   shapeName=dyn_shoelace3_r_capsule
   center=<0.6602 -0.1031 -0.1261>
   rotation=<rot 0 85.8066 -81.8889>
  }]
 dyn_shoelace4_r [{
   shapeType=Capsule
   length=2.864
   radius=1.3414
   shapeName=dyn_shoelace4_r_capsule
   center=<3.3769 0.0338 -0.9969>
   rotation=<rot 0 -79.9918 -80.0902>
  }]
 dyn_shoelace_base_r [{
   shapeType=Capsule
   length=1.5744
   radius=1.2098
   shapeName=dyn_shoelace_base_r_capsule
   center=<-0.6484 -0.1741 -0.255>
   rotation=<rot 0 99.0051 -68.0325>
  }]
}
constraints [
 {bone1Name=dyn_shoelace1_r, bone2Name=dyn_shoelace_base_r, swing1LimitDegrees=20, swing2LimitDegrees=20, twistLimitDegrees=5}
 {bone1Name=dyn_shoelace2_r, bone2Name=dyn_shoelace1_r, swing1LimitDegrees=20, swing2LimitDegrees=20, twistLimitDegrees=5}
 {bone1Name=dyn_shoelace3_r, bone2Name=dyn_shoelace_base_r, swing1LimitDegrees=20, swing2LimitDegrees=20, twistLimitDegrees=5}
 {bone1Name=dyn_shoelace4_r, bone2Name=dyn_shoelace3_r, swing1LimitDegrees=20, swing2LimitDegrees=20, twistLimitDegrees=5}
]
